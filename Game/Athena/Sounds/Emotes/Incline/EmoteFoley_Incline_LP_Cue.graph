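# Sound cue for the Incline emote in Athena, layering a start sound with a looping main sound.
# A Mixer plays the start wave immediately alongside the looping wave, which enters after a fixed half-second delay.
# It uses the emote music sound class and concurrency with Athena emote attenuation, keeps playing when silent, and sends to the music audio-analysis submix.
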SoundCue /Game/Athena/Sounds/Emotes/Incline/EmoteFoley_Incline_LP_Cue
root=Mixer_0
nodes [
 {id=Delay_0, children [WavePlayer_1], properties {delayMax=0.5, delayMin=0.5}}
 {id=Mixer_0, children [WavePlayer_0,Delay_0], properties {inputVolume [1,1]}}
 {id=WavePlayer_0, properties {soundWaveAssetPtr=&/Emote_Incline_Start}}
 {id=WavePlayer_1, properties {bLooping=true, soundWaveAssetPtr=&/Emote_Incline_Main_LP}}
]
properties {
 attenuationSettings=&Fort_Emotes_Athena_Att
 maxDistance=1600
 soundClassObject=&Fort_Music_Emote
 virtualizationMode=PlayWhenSilent
 volumeMultiplier=1
 concurrencyOverrides {maxCount=3}
 concurrencySet [/Game/Sounds/Concurrency_Rules/Athena_EmoteMusic_Conc]
 soundSubmixSends [{minSendLevel=0, soundSubmix=&Fortnite_AudioAnalysis_Music}]
}
